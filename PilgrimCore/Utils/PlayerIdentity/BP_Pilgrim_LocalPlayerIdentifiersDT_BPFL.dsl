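; Blueprint function library that supplies player identity colors and info for Pilgrim (Festival Main Stage), by local player index or Player State.
; A private lookup maps indices 0–3 to rows P1–P4 in DT_Pilgrim_LocalPlayerIdentifiers.
; Player State queries use the splitscreen index in local multiplayer, otherwise overlay the squad pin color on the P1 default.
; Invalid inputs and failed lookups raise script errors with stack traces.
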(fn GetLocalPlayerIdentityColorInfoDefault
  (bind _pilgrim_local_player_identifier_row (Private|GetLocalPlayerIdentityDataTableRowbyLocalPlayerIndex 0))
  (return _pilgrim_local_player_identifier_row))

(fn GetLocalPlayerIdentityDataTableRowByLocalPlayerIndex
  (params (Local%20Player%20Index :type Integer))
  (if (and (>= Local%20Player%20Index 0) (< Local%20Player%20Index 4))
    (bind _returnvalue (Utilities|GetDataTableRow "/PilgrimCore/Utils/PlayerIdentity/DT_Pilgrim_LocalPlayerIdentifiers.DT_Pilgrim_LocalPlayerIdentifiers" (Utilities|Select :Index Local%20Player%20Index :Option%200 "P1" :Option%201 "P2" :Option%202 "P3" :Option%203 "P4"))
      (:then
        (return _returnvalue))
      (:RowNotFound
        (Development|Editor|StackTrace)
        (Class|FortKismetLibrary|ScriptError "Did the Row Names change in the Data Table?")))
    (else
      (Development|Editor|StackTrace)
      (Utilities|Debugging|RaiseScriptError "Invalid Local Player Index. Festival Main Stage only supports 4 local players. Index must be 0, 1, 2, or 3.")))
  (bind _pilgrim_local_player_identifier_row (Private|GetLocalPlayerIdentityDataTableRowbyLocalPlayerIndex 0))
  (return _pilgrim_local_player_identifier_row))

(fn GetPilgrimLocalPlayerIdentityColorByIndex
  (params (LocalPlayerIndex :type Integer))
  (bind _pilgrim_local_player_identifier_row (Private|GetLocalPlayerIdentityDataTableRowbyLocalPlayerIndex LocalPlayerIndex))
  (return (Utilities|Struct|BreakSPilgrimLocalPlayerIdentifierDefinition _pilgrim_local_player_identifier_row)))

(fn GetPilgrimLocalPlayerIdentityColorInfoByIndex
  (params (LocalPlayerIndex :type Integer))
  (bind _pilgrim_local_player_identifier_row (Private|GetLocalPlayerIdentityDataTableRowbyLocalPlayerIndex LocalPlayerIndex))
  (return _pilgrim_local_player_identifier_row))

(fn GetPilgrimPlayerIdentityColorByPlayerState
  (params (Player%20State :type Player%20State%20Object%20Reference))
  (return (Utilities|Struct|BreakSPilgrimLocalPlayerIdentifierDefinition (UI|Pilgrim|GetPilgrimPlayerIdentityColorInfobyPlayerState Player%20State))))

(fn GetPilgrimPlayerIdentityColorInfoByPlayerState
  (params (Player%20State :type Player%20State%20Object%20Reference))
  (bind _returnvalue (PlayerState|GetPlayerController Player%20State))
  (if (Class|PilgrimCoreBPFL|IsLocalPlayerinLocalMPSession Player%20State)
    ; Use Local Player Color
    (Utilities|IsValid _returnvalue
      (:Is%20Valid
        (return (UI|Pilgrim|GetPilgrimLocalPlayerIdentityColorInfobyIndex (Class|FortKismetLibrary|GetSplitscreenIndex _returnvalue))))
      (:Is%20Not%20Valid
        (Development|Editor|StackTrace)
        (Utilities|Debugging|RaiseScriptError "Could not get a player controller ")
        (return (Private|GetLocalPlayerIdentityColorInfoDefault 0))))
    (else
      ; Use Squad Color
      (bind _asfort_player_state_athena (Utilities|Casting|CastToFortPlayerStateAthena Player%20State)
        (:then
          (bind _structout (Utilities|Struct|SetmembersinSPilgrimLocalPlayerIdentifierDefinition (Private|GetLocalPlayerIdentityColorInfoDefault 0) (Class|FortPlayerStateZone|GetPinColor _asfort_player_state_athena)))
          (return _structout))
        (:CastFailed
          (Development|Editor|StackTrace)
          (Utilities|Debugging|RaiseScriptError "Player was neither a local player in local MP nor had a FortPlayerStateAthena.")
          (return (Private|GetLocalPlayerIdentityColorInfoDefault 0)))))))
